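; Creative Spire Spike device: a damageable spike that charges a knockback shockwave from player proximity, damage or events.
; User options and Verse setters update runtime values and scalable floats, while abilities fire or cancel the shockwave under a charge timer.
; Death avoids actor destruction: it clears replicated IsSpawned, optionally drops loot and broadcasts an event.
; Minigame state switches edit and play modes, with an animated range sphere shown outside play.
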(event OnAddedtoMinigame(FortMinigameProgress)
  (params (Minigame :type Fort%20Minigame%20Object%20Reference))
  ;[0.00254 0.2031 0 0.3739] Init
  (CallFunction|SettoEditMode))

(event OnAnyPropertyChanged(ToyOptionsComponent)
  ;[0.00254 0.2031 0 0.3739] Init
  (CallFunction|SetMaxHealth (Variables|UserOptions|Basic|GetMaxHealth))
  (Health|SetHealth (Variables|UserOptions|Basic|GetMaxHealth))
  (CallFunction|SetItemGranting (Variables|UserOptions|Advanced|GetItemGranting))
  (Knockback|Proximity|SetKnockbackChargesfromPlayerProximity (Variables|UserOptions|Basic|GetKnockbackChargesFromPlayerProximity))
  (Knockback|Damaged|SetTakingDamageStartKnockback (Variables|UserOptions|Basic|GetKnockbackChargesFromTakingDamage))
  (Initialization|Initialize (Variables|Default|GetCreative_UseRestriction_Component) (Variables|Default|GetFortMinigameProgress) 0 (Variables|Default|GetToyOptionsComponent) (Variables|UserOptions|Advanced|GetDamageableByTeam) (Variables|UserOptions|Advanced|GetInvertTeamSelection) (Variables|UserOptions|Advanced|GetDamageableByClass) (Variables|UserOptions|Advanced|GetInvulnerable) false)
  (Default|UpdateVisualizationState)
  (Knockback|SetHorizontalKnockbackMultiplier (Variables|UserOptions|Advanced|GetHorizontalKnockbackMultiplier))
  (Knockback|SetVerticalKnockbackMultiplier (Variables|UserOptions|Advanced|GetVerticalKnockbackMultiplier))
  (Knockback|SetAdditionalVerticalKnockback (Variables|UserOptions|Advanced|GetAdditionalVerticalKnockback))
  (Knockback|SetKnockbackTimeLength (Variables|UserOptions|Advanced|GetKnockbackChargeTime))
  (Knockback|SetShouldDestroyStructures (Variables|UserOptions|Advanced|GetShouldDestroyStructures))
  (Knockback|SetKnockbackTriggerRange (Variables|UserOptions|Advanced|GetKnockbackTriggerRange)))

(event OnPropertiesLoaded(ToyOptionsComponent)
  ;[0.00254 0.2031 0 0.3739] Init
  (CallFunction|SetMaxHealth (Variables|UserOptions|Basic|GetMaxHealth))
  (Health|SetHealth (Variables|UserOptions|Basic|GetMaxHealth))
  (CallFunction|SetItemGranting (Variables|UserOptions|Advanced|GetItemGranting))
  (Knockback|Proximity|SetKnockbackChargesfromPlayerProximity (Variables|UserOptions|Basic|GetKnockbackChargesFromPlayerProximity))
  (Knockback|Damaged|SetTakingDamageStartKnockback (Variables|UserOptions|Basic|GetKnockbackChargesFromTakingDamage))
  (Initialization|Initialize (Variables|Default|GetCreative_UseRestriction_Component) (Variables|Default|GetFortMinigameProgress) 0 (Variables|Default|GetToyOptionsComponent) (Variables|UserOptions|Advanced|GetDamageableByTeam) (Variables|UserOptions|Advanced|GetInvertTeamSelection) (Variables|UserOptions|Advanced|GetDamageableByClass) (Variables|UserOptions|Advanced|GetInvulnerable) false)
  (Default|UpdateVisualizationState)
  (Knockback|SetHorizontalKnockbackMultiplier (Variables|UserOptions|Advanced|GetHorizontalKnockbackMultiplier))
  (Knockback|SetVerticalKnockbackMultiplier (Variables|UserOptions|Advanced|GetVerticalKnockbackMultiplier))
  (Knockback|SetAdditionalVerticalKnockback (Variables|UserOptions|Advanced|GetAdditionalVerticalKnockback))
  (Knockback|SetKnockbackTimeLength (Variables|UserOptions|Advanced|GetKnockbackChargeTime))
  (Knockback|SetShouldDestroyStructures (Variables|UserOptions|Advanced|GetShouldDestroyStructures))
  (Knockback|SetKnockbackTriggerRange (Variables|UserOptions|Advanced|GetKnockbackTriggerRange)))

(event MinigameStateChangedDelegate(FortMinigameProgress)
  (params (Minigame :type Fort%20Minigame%20Object%20Reference) (NewMinigameState :type EFortMinigameState%20Enum :default "PreGame"))
  ;[0.00254 0.2031 0 0.3739] Init
  (switch Utilities|FlowControl|Switch|SwitchonEFortMinigameState NewMinigameState
    (:PreGame
      _)
    (:InProgress
      (CallFunction|SettoPlayMode))
    (:PostGameTimeDilation
      _))
  (CallFunction|SettoEditMode))

(event EventBeginPlay
  ;[0.00254 0.2031 0 0.3739] Init
  (DataRegistry|InitializeDataRegistryResolverScopes)
  (|CallParentFunction)
  (Networking|SwitchHasAuthority
    (:Authority
      (Ability|EquipFortAbilitySet self "/CRD_SpireSpike/Abilities/AS_SpireSpikeDevice.AS_SpireSpikeDevice") ; Inherent Ability Sets bugged when spawning from Scenario
    )))

(event EventOnDamageServer
  (params (Damage :type Float%20%28single-precision%29) (DamageTags :type Gameplay%20Tag%20Container%20Structure%20%28by%20ref%29) (Momentum :type Vector :default "0, 0, 0") (HitInfo :type Hit%20Result%20Structure%20%28by%20ref%29) (InstigatedBy :type Controller%20Object%20Reference) (DamageCauser :type Actor%20Object%20Reference) (EffectContext :type Gameplay%20Effect%20Context%20Handle%20Structure))
  (bind _returnvalue (Pawn|GetControlledPawn InstigatedBy))
  ;[0.15 0.0982 0.04092 0.5] Handle Damage: Stop Actor from being Destroyed when killed
  (|CallParentFunction Damage DamageTags Momentum HitInfo InstigatedBy DamageCauser EffectContext)
  (if (not (Health|HasHealthLeft))
    (Utilities|IsValid InstigatedBy
      (:Is%20Valid
        (Utilities|IsValid _returnvalue
          (:Is%20Valid
            (bind _asfort_player_pawn (Utilities|Casting|CastToFortPlayerPawn _returnvalue)
              (:then
                (CallFunction|HandleDeathServer _asfort_player_pawn HitInfo))
              (:CastFailed
                _)))
          (:Is%20Not%20Valid
            _))
        (CallFunction|HandleDeathServer 0 HitInfo))
      (:Is%20Not%20Valid
        _))))

(event EventOnDeathServer
  (params (Damage :type Float%20%28single-precision%29) (DamageTags :type Gameplay%20Tag%20Container%20Structure%20%28by%20ref%29) (Momentum :type Vector :default "0, 0, 0") (HitInfo :type Hit%20Result%20Structure%20%28by%20ref%29) (InstigatedBy :type Controller%20Object%20Reference) (DamageCauser :type Actor%20Object%20Reference) (EffectContext :type Gameplay%20Effect%20Context%20Handle%20Structure))
  ;[0 0 0 0.5] Block Death Effect on Silent Death
  (if (not (GameplayTags|HasTag DamageTags "(TagName=\"Gameplay.Effect.InstantDeath.Silent\")"))
    (|CallParentFunction Damage DamageTags Momentum HitInfo InstigatedBy DamageCauser EffectContext)))

(event OnComponentBeginOverlap(VisualizationRange)
  (params (OverlappedComponent :type Primitive%20Component%20Object%20Reference) (OtherActor :type Actor%20Object%20Reference) (OtherComp :type Primitive%20Component%20Object%20Reference) (OtherBodyIndex :type Integer) (bFromSweep :type Boolean) (SweepResult :type Hit%20Result%20Structure%20%28by%20ref%29))
  ;[0.02684 0.15 0.1457 0.5] Visual Range
  (if (Math|Boolean|ANDBoolean (HelperFunctions|IsLocalPawn OtherActor) (Variables|UserOptions|Advanced|GetShowVisualizationRange) (not (Building|IsPreviewBuildingActor)))
    (CallFunction|ShowRangeVisual)))

(event OnComponentEndOverlap(VisualizationRange)
  (params (OverlappedComponent :type Primitive%20Component%20Object%20Reference) (OtherActor :type Actor%20Object%20Reference) (OtherComp :type Primitive%20Component%20Object%20Reference) (OtherBodyIndex :type Integer))
  ;[0.02684 0.15 0.1457 0.5] Visual Range
  (if (Math|Boolean|ANDBoolean (HelperFunctions|IsLocalPawn OtherActor) (Variables|UserOptions|Advanced|GetShowVisualizationRange) (not (Building|IsPreviewBuildingActor)))
    (CallFunction|HideRangeVisual false)))

(event Custom|ShowRangeVisual
  ;[0.02684 0.15 0.1457 0.5] Visual Range
  (Development|SetHiddeninGame (Variables|Default|GetRangeVisualization) false false)
  (bind _material_instance_dynamic (Default|GetVisualizationSphereMID))
  (bind (_direction _lerp) (Timeline|Timeline Play)
    (:Update
      (Rendering|Material|SetScalarParameterValue (stale Custom|HideRangeVisual _material_instance_dynamic) "Opacity" _lerp :!class /Script/Engine.MaterialInstanceDynamic)
      (Transformation|SetWorldScale3D (Variables|Default|GetRangeVisualization) (Math|Conversions|ToVector(Float) (Math|Float|Lerp 0.0 (* (Variables|UserOptions|Advanced|GetKnockbackTriggerRange) 2.0) _lerp))))
    (:Finished
      (switch Utilities|FlowControl|Switch|SwitchonETimelineDirection _direction
        (:Backward
          (Development|SetHiddeninGame (Variables|Default|GetRangeVisualization) true false))))))

(event Custom|HideRangeVisual
  (params (Immediate :type Boolean))
  ;[0.02684 0.15 0.1457 0.5] Visual Range
  (bind _material_instance_dynamic (Default|GetVisualizationSphereMID))
  (if (not Immediate)
    (Timeline|Timeline Reverse)
    (else
      (Timeline|Timeline SetNewTime))))

(event Custom|SpawnEvent
  ;[1 1 1] Spawn/Despawn
  (CallFunction|SpawnFunction))

(event Custom|DespawnEvent
  ;[1 1 1] Spawn/Despawn
  (CallFunction|DespawnFunction))

(event Custom|DestroyEvent
  (params (Instigator :type Controller%20Object%20Reference))
  ;[1 1 1] Destroy
  (CallFunction|DestroyFunction Instigator))

(event Custom|KnockbackEvent
  ;[1 1 1] Knockback
  (Knockback|KnockbackImmediateFunction))

(event Custom|ChargeKnockbackEvent
  ;[1 1 1] Knockback
  (Knockback|ForceChargeKnockback))

(event Custom|EnableItemGrantingEvent
  ;[1 1 1] Item Granting
  (CallFunction|SetItemGranting true))

(event Custom|DisableItemGrantingEvent
  ;[1 1 1] Item Granting
  (CallFunction|SetItemGranting false))

(event Custom|EnableInvulnerabilityEvent
  ;[1 1 1] Invulnerability
  (Default|SetInvulnerability true))

(event Custom|DisableInvulnerabilityEvent
  ;[1 1 1] Invulnerability
  (Default|SetInvulnerability false))

(fn BlueprintCanInteract
  (params (InteractingPawn :type Fort%20Pawn%20Object%20Reference) (InteractionBeingAttempted :type EInteractionBeingAttempted%20Enum :default "FirstInteraction") (InteractionType :type TInteractionType%20Enum :default "IT_NoInteraction"))
  (return (Interaction|CanEditOptionsComponent (Variables|Default|GetToyOptionsComponent) InteractingPawn)))

(fn BlueprintModifyIncomingDamage
  (params (Damage :type Float%20%28single-precision%29) (DamageTags :type Gameplay%20Tag%20Container%20Structure%20%28by%20ref%29) (EffectContext :type Gameplay%20Effect%20Context%20Handle%20Structure%20%28by%20ref%29) (EventInstigator :type Controller%20Object%20Reference) (DamageCauser :type Actor%20Object%20Reference))
  (bind _valid_controller (DEPRECATED|ActorHasValidController (Variables|Default|GetCreative_UseRestriction_Component) EventInstigator true true))
  (if _valid_controller
    (bind _returnvalue (|CallParentFunction Damage DamageTags EffectContext EventInstigator DamageCauser))
    (return _returnvalue)
    (else
      (return 0.0))))

(fn GetExecutionContext
  (return "ServerOnly"))

(fn Pause
  (return false))

(fn Play
  (return false))

(fn Stop
  (return false))

(fn BuildDataRegistryResolverScope
  (return (Variables|DataRegistry|GetResolver%20Scopes) 1 true))

(fn CanDoKnockback
  (return (and (Variables|Default|GetIsSpawned) (Class|CreativeCommonDeviceFunctionLibrary|IsinPlayMode self))))

(fn ClearKnockbackChargeTimer
  (Utilities|Time|ClearandInvalidateTimerbyHandle (Variables|Knockback|GetKnockbackTimerHandle)))

(fn DespawnFunction
  (Networking|SwitchHasAuthority
    (:Authority
      (Variables|Default|SetIsSpawned false))))

(fn DestroyFunction
  (params (Instigator :type Controller%20Object%20Reference))
  (bind _returnvalue (Pawn|GetControlledPawn Instigator))
  (if (Variables|Default|GetIsSpawned)
    (Utilities|IsValid Instigator
      (:Is%20Valid
        (Utilities|IsValid _returnvalue
          (:Is%20Valid
            (bind _asfort_player_pawn (Utilities|Casting|CastToFortPlayerPawn _returnvalue)
              (:then
                (CallFunction|HandleDeathServer _asfort_player_pawn))
              (:CastFailed
                _)))
          (:Is%20Not%20Valid
            _))
        (CallFunction|HandleDeathServer 0))
      (:Is%20Not%20Valid
        _))))

(fn DoesTakingDamageRestartKnockbackTimer
  (return (Variables|Knockback|Melee|GetTakingDamageStartKnockbackRuntime)))

(fn DummyEvent)

(fn ForceChargeKnockback
  (if (Knockback|CanDoKnockback)
    (Knockback|StopKnockbackAbility)
    (bind _returnvalue (Abilities|TryActivateAbilitybyClass (Ability|GetFortAbilitySystemComponent self) (Variables|Default|GetShockwaveGA) true))
    (if _returnvalue
      (if (not (Utilities|Time|IsValidTimerHandle (Variables|Knockback|GetKnockbackTimerHandle)))
        ;[0.15 0.05974 0.02516 0.5] Set timer temporarily so that the Get function can return the time to verse
        (bind _returnvalue_2 (Utilities|Time|SetTimerbyEvent (EventDispatchers|CreateEvent :!event DummyEvent) (Knockback|GetKnockbackTime)))
        (Variables|Knockback|SetKnockbackTimerHandle _returnvalue_2)))))

(fn GetAdditionalVerticalKnockbackVerse
  (return (Variables|Knockback|GetAdditionalVerticalKnockbackRuntime)))

(fn GetCacheContextOverride
  (return 0))

(fn GetCanGrantItemsVerse
  (return (Variables|Default|GetItemGrantingRuntime)))

(fn GetCurrentHealthVerse
  (if (Variables|Default|GetIsSpawned)
    (return (Health|GetHealth))
    (else
      (return 0.0))))

(fn GetHorizontalKnockbackMultiplierVerse
  (return (Variables|Knockback|GetHorizontalKnockbackMultiplierRuntime)))

(fn GetInvulnerableVerse
  (return (not (Variables|Actor|GetbCanBeDamaged))))

(fn GetKnockbackChargeTimerLengthVerse
  (return (Variables|Knockback|GetKnockbackTimeLengthRuntime)))

(fn GetKnockbackChargeTimerVerse
  (return (Utilities|Time|GetTimerRemainingTimebyHandle (Variables|Knockback|GetKnockbackTimerHandle))))

(fn GetKnockbackTime
  (return (Math|Float|Max(Float) (Math|Float|SelectFloat 0.0 (Variables|Knockback|GetKnockbackTimeLengthRuntime) (Variables|Knockback|GetImmediateKnockback)) 0.01)))

(fn GetKnockbackTriggerRangeVerse
  (return (Variables|Knockback|GetKnockbackTriggerRangeRuntime)))

(fn GetMaxHealthVerse
  (return (Health|GetMaxHealth)))

(fn GetOverrideKnockbackTimerHandle
  (return (Variables|Knockback|GetKnockbackTimerHandle)))

(fn GetShouldChargeKnockbackOnPlayerProximityVerse
  (return (Variables|Default|GetUseProximityActivation)))

(fn GetShouldKnockbackChargeFromTakingDamageVerse
  (return (Variables|Knockback|Melee|GetTakingDamageStartKnockbackRuntime)))

(fn GetShouldKnockbackDestroyStructuresVerse
  (return (Variables|Knockback|GetShouldDestroyStructuresRuntime)))

(fn GetVerticalKnockbackMultiplierVerse
  (return (Variables|Knockback|GetVerticalKnockbackMultiplierRuntime)))

(fn GetVisualizationSphereMID
  (bind _asmaterial_instance_dynamic (Utilities|Casting|CastToMaterialInstanceDynamic (Rendering|Material|GetMaterial (Variables|Default|GetRangeVisualization) 0))
    (:then
      (return _asmaterial_instance_dynamic))
    (:CastFailed
      (Material|ReattachOrCreateMID (Variables|Default|GetRangeVisualization) (LocalVariables|GetVisualMID))
      (return (LocalVariables|GetVisualMID)))))

(fn HandleDeathServer
  (params (Instigator :type Fort%20Player%20Pawn%20Object%20Reference) (HitResult :type Hit%20Result%20Structure%20%28by%20ref%29))
  (bind _ondestroy (Variables|UserOptions|Events|GetOnDestroy))
  (Networking|SwitchHasAuthority
    (:Authority
      (if (not (Variables|Default|GetbWasDeathHandled))
        (Variables|Default|SetbWasDeathHandled true)
        (Utilities|FlowControl|Sequence
          (:then_0
            (GameplayCue|ExecuteGameplayCueOnActor(Burst) self (Variables|Default|GetGC_Death) (GameplayCue|MakeGameplayCueParametersfromHitResult HitResult))
            (Variables|Default|SetIsSpawned false)
            (Health|SetHealth 1.0)
            (if (Variables|Default|GetItemGrantingRuntime)
              (Container|SpawnLoot (Variables|Actor|GetInstigator))))
          (:then_1
            (Utilities|IsValid (Variables|Actor|GetInstigator)
              (:Is%20Valid
                (GameplayEvent|BroadcastEvent _ondestroy 0 (Utilities|Struct|MakeGameplayEventSpireSpikeonDestroyed (Pawn|GetController (Variables|Actor|GetInstigator)))))
              (:Is%20Not%20Valid
                (GameplayEvent|BroadcastEvent _ondestroy 0 (Utilities|Struct|MakeGameplayEventSpireSpikeonDestroyed 0))))))))))

(fn InitializeDataRegistryResolverScopes
  (Networking|SwitchHasAuthority
    (:Authority
      (if (Utilities|Array|IsEmpty (Variables|DataRegistry|GetResolver%20Scopes))
        (Variables|DataRegistry|SetResolver%20Scopes (Utilities|Array|MakeArray (SpatialGameplay|GetSpatialGameplayInterfaceFNameforActor self) (Utilities|String|StringToName (Utilities|GetDisplayName self))))
        (return)))))

(fn IsCharging
  (return (Utilities|Time|IsValidTimerHandle (Variables|Knockback|GetKnockbackTimerHandle))))

(fn IsSpawnedVerse
  (return (Variables|Default|GetIsSpawned)))

(fn KnockbackImmediateFunction
  (if (Knockback|CanDoKnockback)
    (Knockback|StopKnockbackAbility)
    (Variables|Knockback|SetImmediateKnockback true)
    (Abilities|TryActivateAbilitybyClass (Ability|GetFortAbilitySystemComponent self) (Variables|Default|GetShockwaveGA) true)
    (Variables|Knockback|SetImmediateKnockback false)))

(fn OnKnockbackFunction
  (GameplayEvent|BroadcastEvent (Variables|UserOptions|Events|GetOnKnockback) 0 (Utilities|Struct|MakeGameplayEventSpireSpikeonKnockback 0))
  (Variables|Default|SetPickaxeActivated false))

(fn OnRep_IsSpawned
  (Utilities|FlowControl|Sequence
    (:then_0
      (Networking|SwitchHasAuthority
        (:Authority
          (if (Variables|Default|GetIsSpawned)
            (Default|RegisterNearbyActorsPerception)
            (else
              (Knockback|StopKnockbackAbility)
              (CallFunction|UnregisterNearbyActorsPerception))))))
    (:then_1
      (CallFunction|SetSuppressHealthBar (not (Variables|Default|GetIsSpawned)))
      (Collision|SetActorEnableCollision (Variables|Default|GetIsSpawned))
      (Networking|CanExecuteCosmeticEvents
        (:True
          (Rendering|SetVisibility (Variables|Debug|GetStaticMeshComponent) (Variables|Default|GetIsSpawned) true))))))

(fn RegisterNearbyActorsPerception
  (if (Class|CreativeCommonDeviceFunctionLibrary|IsinPlayMode self)
    (|CallParentFunction)))

(fn SetAdditionalVerticalKnockback
  (params (NewValue :type Float%20%28double-precision%29))
  (bind _returnvalue (Math|Float|Clamp(Float) NewValue 0.0 1e+02))
  (if (!= _returnvalue (Variables|Knockback|GetAdditionalVerticalKnockbackRuntime))
    (bind _output_get (Variables|Knockback|SetAdditionalVerticalKnockbackRuntime _returnvalue))
    (ScalableFloat|SetValuewithContext (Variables|ScalableFloats|GetSF_KnockbackVerticalAdd) (* _output_get 1e+02))))

(fn SetCurrentHealthVerse
  (params (NewValue :type Float%20%28double-precision%29))
  (if (> NewValue 0)
    (Health|SetHealth (Variables|GetNewValue))
    (else
      (CallFunction|DestroyEvent 0))))

(fn SetHorizontalKnockbackMultiplier
  (params (NewValue :type Float%20%28double-precision%29))
  (bind _returnvalue (Math|Float|Clamp(Float) NewValue 0.0 1e+02))
  (if (!= _returnvalue (Variables|Knockback|GetHorizontalKnockbackMultiplierRuntime))
    (bind _output_get (Variables|Knockback|SetHorizontalKnockbackMultiplierRuntime _returnvalue))
    (ScalableFloat|SetValuewithContext (Variables|ScalableFloats|GetSF_KnockbackHorizonalForce) (* _output_get 1e+02))))

(fn SetInvulnerability
  (params (Invulnerable :type Boolean))
  (if Invulnerable
    (Variables|Actor|SetbCanBeDamaged false) ; Invulnerable
    (Navigation|SetNavigationIndestructible true)
    (else
      (Variables|Actor|SetbCanBeDamaged true) ; Vulnerable
      (Navigation|SetNavigationIndestructible false)))
  (AI|RequestNavUpdateforBuilding self))

(fn SetItemGranting
  (params (NewValue :type Boolean))
  (if (!= NewValue (Variables|Default|GetItemGrantingRuntime))
    (Variables|Default|SetItemGrantingRuntime (Variables|GetNewValue))))

(fn SetKnockbackChargesFromPlayerProximity
  (params (NewValue :type Boolean))
  (if (!= NewValue (Variables|Default|GetUseProximityActivation))
    (bind _output_get (Variables|Default|SetUseProximityActivation (Variables|GetNewValue)))
    (if _output_get
      (Default|RegisterNearbyActorsPerception)
      (else
        (CallFunction|UnregisterNearbyActorsPerception)))))

(fn SetKnockbackChargeTimerVerse
  (params (NewValue :type Float%20%28double-precision%29))
  (Variables|Knockback|SetKnockbackTimeLengthRuntime (Math|Float|Clamp(Float) NewValue 0.0 6e+01))
  (Knockback|ForceChargeKnockback))

(fn SetKnockbackTimeLength
  (params (KnockbackTimeLengthRuntime :type Float%20%28double-precision%29))
  (Variables|Knockback|SetKnockbackTimeLengthRuntime (Math|Float|Clamp(Float) KnockbackTimeLengthRuntime 0.0 6e+01))
  (if (Knockback|CanDoKnockback)
    (if (Knockback|IsCharging)
      (Knockback|ForceChargeKnockback))))

(fn SetKnockbackTriggerRange
  (params (NewValue :type Float%20%28double-precision%29))
  (bind _returnvalue (Math|Float|Clamp(Float) NewValue 2.0 1e+01))
  (if (!= _returnvalue (Variables|Knockback|GetKnockbackTriggerRangeRuntime))
    (bind _output_get (Variables|Knockback|SetKnockbackTriggerRangeRuntime _returnvalue))
    (ScalableFloat|SetValuewithContext (Variables|ScalableFloats|GetSF_KnockbackRange) (* _output_get 1e+02))
    (ScalableFloat|SetValuewithContext (Variables|Default|GetPerceptionRadius) (* _output_get 1e+02))
    (Default|RegisterNearbyActorsPerception)
    (Networking|CanExecuteCosmeticEvents
      (:True
        (Transformation|SetWorldScale3D (Variables|Default|GetRangeVisualization) (Math|Conversions|ToVector(Float) (* (Variables|GetNewValue) 2.0)))))))

(fn SetMaxHealth
  (params (NewValue :type Float%20%28double-precision%29))
  (bind _returnvalue (Math|Float|Max(Float) NewValue 1.0))
  (if (!= (Variables|Default|GetMaxHealthRuntime) _returnvalue)
    (bind _output_get (Variables|Default|SetMaxHealthRuntime _returnvalue))
    (Health|SetMaxHealthDontCreateASC _output_get)))

(fn SetOverrideKnockbackTimerHandle
  (params (New%20Value :type Timer%20Handle%20Structure))
  (Utilities|Time|ClearandInvalidateTimerbyHandle (Variables|Knockback|GetKnockbackTimerHandle)) ; Clear the temp timer now that the main one has been set
  (Variables|Knockback|SetKnockbackTimerHandle New%20Value))

(fn SetShouldDestroyStructures
  (params (NewValue :type Boolean))
  (if (!= NewValue (Variables|Knockback|GetShouldDestroyStructuresRuntime))
    (bind _output_get (Variables|Knockback|SetShouldDestroyStructuresRuntime NewValue))
    (ScalableFloat|SetValuewithContext (Variables|ScalableFloats|GetSF_KnockbackDestoyStructures) (Math|Float|SelectFloat 1.0 0.0 _output_get))))

(fn SetTakingDamageStartKnockback
  (params (NewValue :type Boolean))
  (if (!= NewValue (Variables|Knockback|Melee|GetTakingDamageStartKnockbackRuntime))
    (Variables|Knockback|Melee|SetTakingDamageStartKnockbackRuntime (Variables|GetNewValue))))

(fn SetToEditMode
  (Networking|SwitchHasAuthority
    (:Authority
      (Default|SetInvulnerability true)
      (CallFunction|SpawnFunction)
      (CallFunction|UnregisterNearbyActorsPerception)
      (Knockback|StopKnockbackAbility))
    (:Remote
      (Default|UpdateVisualizationState))))

(fn SetToPlayMode
  (Networking|SwitchHasAuthority
    (:Authority
      (Utilities|FlowControl|Sequence
        (:then_0
          (Default|SetInvulnerability (Variables|UserOptions|Advanced|GetInvulnerable))
          (if (Variables|UserOptions|Basic|GetSpawnAtGameStart)
            (CallFunction|SpawnFunction)
            (else
              (CallFunction|DespawnFunction))))
        (:then_1
          (Knockback|StopKnockbackAbility))))
    (:Remote
      (Default|UpdateVisualizationState))))

(fn SetVerticalKnockbackMultiplier
  (params (NewValue :type Float%20%28double-precision%29))
  (bind _returnvalue (Math|Float|Clamp(Float) NewValue 0.0 1e+02))
  (if (!= _returnvalue (Variables|Knockback|GetVerticalKnockbackMultiplierRuntime))
    (bind _output_get (Variables|Knockback|SetVerticalKnockbackMultiplierRuntime _returnvalue))
    (ScalableFloat|SetValuewithContext (Variables|ScalableFloats|GetSF_KnockbackVerticalForce) (* _output_get 1e+02))))

(fn SpawnFunction
  (Networking|SwitchHasAuthority
    (:Authority
      (Health|SetHealth (Variables|Default|GetMaxHealthRuntime))
      (Variables|Default|SetbWasDeathHandled false)
      (Variables|Default|SetIsSpawned true)
      (Variables|Container|SetbAlreadySearched false)
      (Knockback|Proximity|TryStartProximityKnockback))))

(fn StartProximityKnockback
  (if (and (Knockback|CanDoKnockback) (Variables|Default|GetUseProximityActivation))
    (Abilities|TryActivateAbilitybyClass (Ability|GetFortAbilitySystemComponent self) (Variables|Default|GetShockwaveGA) true)))

(fn StartTakingDamageKnockback
  (params (Pickaxe :type Boolean))
  (if (Knockback|Damaged|DoesTakingDamageRestartKnockbackTimer)
    (Knockback|StopKnockbackAbility)
    (Abilities|TryActivateAbilitybyClass (Ability|GetFortAbilitySystemComponent self) (Variables|Default|GetShockwaveGA) true)))

(fn StopKnockbackAbility
  (Networking|SwitchHasAuthority
    (:Authority
      (Abilities|TryActivateAbilitybyClass (Ability|GetFortAbilitySystemComponent self) "/CRD_SpireSpike/Abilities/GA_SpireSpikeDevice_CancelShockwave.GA_SpireSpikeDevice_CancelShockwave_C" true)
      (Knockback|ClearKnockbackChargeTimer))))

(fn TryStartProximityKnockback
  (if (> (Variables|Default|GetSensedPlayerPawnCount) 0)
    (Knockback|Proximity|StartProximityKnockback)))

(fn UpdateVisualizationState
  (Networking|CanExecuteCosmeticEvents
    (:True
      (if (and (not (Class|CreativeCommonDeviceFunctionLibrary|IsinPlayMode self)) (Variables|UserOptions|Advanced|GetShowVisualizationRange))
        (Collision|SetCollisionEnabled (Variables|Default|GetVisualizationRange) "QueryOnly")
        (else
          (Collision|SetCollisionEnabled (Variables|Default|GetVisualizationRange) "NoCollision")
          (CallFunction|HideRangeVisual true))))))

(fn ConstructionScript
  (|CallParentFunction)
  (if (Editor|InEditor)
    (Rendering|SetVisibility (Variables|Default|GetRangeVisualization) (Variables|UserOptions|Advanced|GetShowVisualizationRange) false)))

(fn WaitAndTryProximityKnockback
  (Utilities|Time|SetTimerbyEvent (EventDispatchers|CreateEvent :!event TryStartProximityKnockback) (+ (ScalableFloat|GetValueatLevel (Variables|Default|GetKnockbackCooldownBuffer)) (Components|Tick|GetComponentTickInterval (Variables|Default|GetFortNearbyActorsPerception)))) ; Wait for Perception to Tick
)
